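# Niagara system for the RhymeLock emote vehicle, emitting short-lived flame bursts with light at two mirrored offsets.
# Flames spawns in game, gated by owner speed and suppressed by a water-and-depth check, while duplicate Flames_FE spawns only in front-end preview.
# Spawn groups place each side and mirror its light and plate offsets.
# Curves pulse light and plate alpha over each particle's life, feeding mesh and light renderers.
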NiagaraSystem /BRCosmetics/Effects/Fort_Effects/Emotes/RhymeLock/NS_EmoteRhymeLock_Vehicle
dependencies {
 usedDataInterfaces [&NiagaraDataInterfaceCurve]
 usedDynamicInputs [&RandomRangeFloat]
 usedModules [&SystemState]
}
emitters {
 Flames {
  emitterUpdateScript [
   &EmitterState()
   &SpawnRate(
    SpawnRate=&MakeCustomFloatFromBool(
     Bool=$User.bIsFrontEndPreview
     True Float=0
     False Float=&MakeCustomFloatFromBool(
      Bool=&Boolean_AndOperation(Bool1=$User.bIsInWater, Bool2=&SetBoolByFloatComparison(A=$User.WaterDepth, B=30))
      True Float=0
      False Float=&MakeCustomFloatFromBool(
       Bool=&SetBoolByFloatComparison(A=&VectorLength(Vector=$Engine.Owner.Velocity), B=25)
       True Float=5
       False Float=0
      )
     )
    )
    Use Spawn Probability=false
   )
   &SpawnRate(
    SpawnGroup=1
    SpawnRate=&MakeCustomFloatFromBool(
     Bool=$User.bIsFrontEndPreview
     True Float=0
     False Float=&MakeCustomFloatFromBool(
      Bool=&Boolean_AndOperation(Bool1=$User.bIsInWater, Bool2=&SetBoolByFloatComparison(A=$User.WaterDepth, B=30))
      True Float=0
      False Float=&MakeCustomFloatFromBool(
       Bool=&SetBoolByFloatComparison(A=&VectorLength(Vector=$Engine.Owner.Velocity), B=25)
       True Float=5
       False Float=0
      )
     )
    )
    Use Spawn Probability=false
   ) {module=SpawnRate001}
  ]
  particleSpawnScript [
   &InitializeParticle(
    Color Mode=::NewEnumerator0
    Lifetime Max=0.35
    Lifetime Min=0.35
    Lifetime Mode=::NewEnumerator1
    Mass Mode=::NewEnumerator3
    Sprite Rotation Mode=::NewEnumerator2
    Sprite Size Mode=::NewEnumerator1
    Uniform Sprite Size Max=12
    Uniform Sprite Size Min=15
    UsePositionOffset=false
    Write Material Random=true
   )
   &SystemLocation(MaskSpawnGroup=false, Offset=<-14.7 -48.3454 30.1>)
   &SystemLocation(MaskSpawnGroup=true, Offset=<14.7 -48.3454 30.1>, Spawn Group Mask=1) {module=SystemLocation001}
   /SetVariables(
    Particles.LightPos:Vector3f=&Add_Vector(
     A=Linked() {linkedVariable {name=Particles.Position, type=NiagaraPosition}}
     B=&Lerp_Vector(A=<-2 -9 -5>, B=<2 -9 -5>, Alpha=&MaskFloatBySpawnGroup(Masked Float=0, Passthrough Float=1))
    )
    Particles.Plate Pos:Vector3f=&Add_Vector(
     A=Linked() {linkedVariable {name=Particles.Position, type=NiagaraPosition}}
     B=&Lerp_Vector(A=<5.6 -1.35 -2>, B=<-5.6 -1.35 -2>, Alpha=&MaskFloatBySpawnGroup(Masked Float=0, Passthrough Float=1))
    )
    Particles.Ring Color:LinearColor=<rgba 9.1667 0 10 1>
   ) {isSetParameters=true, module=SetVariables_#1}
  ]
  particleUpdateScript [
   &ParticleState()
   &ScaleSpriteSize(Scale Factor=&Vector2DFromFloat(Value=&FloatFromCurve(FloatCurve=DataInterface() {propertyValues {Curve {keys [<at 0 1>,<at 1 1.25>]}}})))
   /SetVariables(
    Particles.Light Color:LinearColor=&MakeLinearColorFromVectorAndFloat(
     "Vector (RGB)"=&MakeVectorFromLinearColorRGB(Color=<rgba 0.5 0.16 0.075 1>)
     "Float (Alpha)"=&FloatFromCurve(
      FloatCurve=DataInterface() {propertyValues {Curve {keys [
          {arriveTangent=4.4844, interpMode=RCIM_Cubic, leaveTangent=4.4844, tangentMode=RCTM_User, time=0, value=0}
          <at 0.5 1 RCIM_Cubic>
          {arriveTangent=-4.8003, interpMode=RCIM_Cubic, leaveTangent=-4.8003, tangentMode=RCTM_User, time=1, value=0}
         ]}}}
      Scale Curve=50
     )
    )
    Particles.LightExponent=3
    Particles.PlateColor:LinearColor=&MakeLinearColorFromVectorAndFloat(
     "Vector (RGB)"=&MakeVectorFromLinearColorRGB(Color=<rgba 3 0.5 0 1>)
     "Float (Alpha)"=&FloatFromCurve(
      FloatCurve=DataInterface() {propertyValues {Curve {keys [
          {arriveTangent=9.766, interpMode=RCIM_Cubic, leaveTangent=5.1913, tangentMode=RCTM_User, time=0, value=0}
          <at 0.5 1 RCIM_Cubic>
          {arriveTangent=-5.6128, interpMode=RCIM_Cubic, leaveTangent=-19.2021, tangentMode=RCTM_User, time=1, value=0}
         ]}}}
      Scale Curve=0.25
     )
    )
   ) {isSetParameters=true, module=SetVariables_#2}
  ]
  rendererClasses [&NiagaraMeshRendererProperties, &NiagaraLightRendererProperties]
  renderers [
   {rendererClass=&NiagaraMeshRendererProperties, rendererIndex=0}
   {rendererClass=&NiagaraLightRendererProperties, rendererIndex=1}
   {rendererClass=&NiagaraMeshRendererProperties, rendererIndex=2}
   {rendererClass=&NiagaraMeshRendererProperties, rendererIndex=3}
  ]
 }
 Flames_FE {
  emitterUpdateScript [
   &EmitterState(
    Enable Distance Culling=true
    Life Cycle Mode=::NewEnumerator1
    Loop Duration=2
    Module.bApplySpawnCountScaleByDistance=false
    Module.bUseMaxDistance=false
    Module.bUseMinDistance=false
    Scalability Mode=::NewEnumerator1
    Scale Spawn Count=true
    UseLoopDelay=false
   )
   &SpawnRate(
    SpawnRate=&MakeCustomFloatFromBool(Bool=$User.bIsFrontEndPreview, True Float=5, False Float=0)
    Use Spawn Probability=false
   )
   &SpawnRate(
    SpawnGroup=1
    SpawnRate=&MakeCustomFloatFromBool(Bool=$User.bIsFrontEndPreview, True Float=5, False Float=0)
    Use Spawn Probability=false
   ) {module=SpawnRate001}
  ]
  particleSpawnScript [
   &InitializeParticle(
    Color Mode=::NewEnumerator0
    Lifetime Max=0.35
    Lifetime Min=0.35
    Lifetime Mode=::NewEnumerator1
    Mass Mode=::NewEnumerator3
    Sprite Rotation Mode=::NewEnumerator2
    Sprite Size Mode=::NewEnumerator1
    Uniform Sprite Size Max=12
    Uniform Sprite Size Min=15
    UsePositionOffset=false
    Write Material Random=true
   )
   &SystemLocation(MaskSpawnGroup=false, Offset=<-14.7 -48.3454 30.1>)
   &SystemLocation(MaskSpawnGroup=true, Offset=<14.7 -48.3454 30.1>, Spawn Group Mask=1) {module=SystemLocation001}
   /SetVariables(
    Particles.LightPos:Vector3f=&Add_Vector(
     A=Linked() {linkedVariable {name=Particles.Position, type=NiagaraPosition}}
     B=&Lerp_Vector(A=<-2 -9 -5>, B=<2 -9 -5>, Alpha=&MaskFloatBySpawnGroup(Masked Float=0, Passthrough Float=1))
    )
    Particles.Plate Pos:Vector3f=&Add_Vector(
     A=Linked() {linkedVariable {name=Particles.Position, type=NiagaraPosition}}
     B=&Lerp_Vector(A=<5.6 -1.35 -2>, B=<-5.6 -1.35 -2>, Alpha=&MaskFloatBySpawnGroup(Masked Float=0, Passthrough Float=1))
    )
    Particles.Ring Color:LinearColor=<rgba 9.1667 0 10 1>
   ) {isSetParameters=true, module=SetVariables_#1}
  ]
  particleUpdateScript [
   &ParticleState()
   &ScaleSpriteSize(Scale Factor=&Vector2DFromFloat(Value=&FloatFromCurve(FloatCurve=DataInterface() {propertyValues {Curve {keys [<at 0 1>,<at 1 1.25>]}}})))
   /SetVariables(
    Particles.Light Color:LinearColor=&MakeLinearColorFromVectorAndFloat(
     "Vector (RGB)"=&MakeVectorFromLinearColorRGB(Color=<rgba 0.5 0.16 0.075 1>)
     "Float (Alpha)"=&FloatFromCurve(
      FloatCurve=DataInterface() {propertyValues {Curve {keys [
          {arriveTangent=4.4844, interpMode=RCIM_Cubic, leaveTangent=4.4844, tangentMode=RCTM_User, time=0, value=0}
          <at 0.5 1 RCIM_Cubic>
          {arriveTangent=-4.8003, interpMode=RCIM_Cubic, leaveTangent=-4.8003, tangentMode=RCTM_User, time=1, value=0}
         ]}}}
      Scale Curve=50
     )
    )
    Particles.LightExponent=3
    Particles.PlateColor:LinearColor=&MakeLinearColorFromVectorAndFloat(
     "Vector (RGB)"=&MakeVectorFromLinearColorRGB(Color=<rgba 3 0.5 0 1>)
     "Float (Alpha)"=&FloatFromCurve(
      FloatCurve=DataInterface() {propertyValues {Curve {keys [
          {arriveTangent=9.766, interpMode=RCIM_Cubic, leaveTangent=5.1913, tangentMode=RCTM_User, time=0, value=0}
          <at 0.5 1 RCIM_Cubic>
          {arriveTangent=-5.6128, interpMode=RCIM_Cubic, leaveTangent=-19.2021, tangentMode=RCTM_User, time=1, value=0}
         ]}}}
      Scale Curve=0.25
     )
    )
   ) {isSetParameters=true, module=SetVariables_#2}
  ]
  rendererClasses [&NiagaraMeshRendererProperties, &NiagaraLightRendererProperties]
  renderers [
   {rendererClass=&NiagaraMeshRendererProperties, rendererIndex=0}
   {rendererClass=&NiagaraLightRendererProperties, rendererIndex=1}
   {rendererClass=&NiagaraMeshRendererProperties, rendererIndex=2}
   {rendererClass=&NiagaraMeshRendererProperties, rendererIndex=3}
  ]
 }
}
summary {
 systemName=NS_EmoteRhymeLock_Vehicle
 emitters [
  {emitterName=Flames, rendererClasses [&NiagaraMeshRendererProperties, &NiagaraLightRendererProperties]}
  {emitterName=Flames_FE, rendererClasses [&NiagaraMeshRendererProperties, &NiagaraLightRendererProperties]}
 ]
 userVariables [
  User.bIsFrontEnd {type=NiagaraBool}
  User.bIsFrontEndPreview {type=NiagaraBool}
  User.bIsInWater {type=NiagaraBool}
  User.WaterDepth {type=NiagaraFloat, defaultValue=0, description=""}
 ]
}
