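# Niagara system for the orange wheel trail of the RhymeLock emote, built from two ribbon emitters switched by User.bIsFrontEndPreview.
# Trail spawns per distance moved in game, while Trail_FE spawns at a steady rate in front-end preview, using velocity to stream the ribbon.
# Both use two ribbon renderers, carry separate "Lines" attributes, and narrow and fade over a short life.
# A platform check raises alpha scale on Switch.
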NiagaraSystem /BRCosmetics/Effects/Fort_Effects/Emotes/RhymeLock/NS_EmoteRhymeLock_WheelTrail
dependencies {
 usedDataInterfaces [&NiagaraDataInterfaceCurve, &NiagaraDataInterfacePlatformSet]
 usedDynamicInputs []
 usedModules [&SystemState]
}
emitters {
 Trail {
  emitterUpdateScript [
   &EmitterState()
   &SpawnPerUnit(
    Module.Use Max Movement Threshold=true
    Module.Use Movement Tolerance=false
    Spawn Per Unit:NiagaraFloat=&MakeCustomFloatFromBool(Bool=$User.bIsFrontEndPreview, True Float=0, False Float=12)
    Use Spawn Probability=false
   )
   &SpawnRate(SpawnRate=900, Use Spawn Probability=false) {enabled=false}
  ]
  particleSpawnScript [
   &InitializeRibbon(
    Color=<rgba 50 6.6667 0 1>
    Lifetime=0.5
    Module.Write Color=true
    Module.Write Lifetime=true
    Module.Write Mass=true
    Module.Write Material Random=true
    Module.Write Position=false
    Module.Write RibbonTwist=true
    Module.Write RibbonWidth=true
    Ribbon Twist=-0.12
    Ribbon Width=25
   )
   /SetVariables(Particles.RibbonDistanceFromStart:NiagaraFloat=$Emitter.DistanceTraveled) {isSetParameters=true, module=SetVariables_#1}
   /SetVariables(Particles.Lines COlor:LinearColor=<rgba 75 10 0 1>, Particles.Lines Twist=-0.35, Particles.Lines Width=20) {isSetParameters=true, module=SetVariables_#2}
   &AddVelocity(Velocity=<0 -500 0>) {enabled=false}
  ]
  particleUpdateScript [
   &SolveForcesAndVelocity(Clamp Velocity=false, Limit Acceleration=false) {enabled=false}
   &ParticleState()
   &ScaleColor(
    Scale Alpha=&FloatFromCurve(
     FloatCurve=DataInterface() {propertyValues {Curve {keys [
         {arriveTangent=5.1472, interpMode=RCIM_Cubic, leaveTangent=5.1044, tangentMode=RCTM_User, time=0, value=0}
         <at 0.5 1 RCIM_Cubic>
         {arriveTangent=-5.1044, interpMode=RCIM_Cubic, leaveTangent=-5.1044, tangentMode=RCTM_User, time=1, value=0}
        ]}}}
     Scale Curve=&MakeCustomFloatFromBool(
      Bool=&IsPlatformSetActive(New Platform Set=DataInterface() {propertyValues {Platforms {qualityLevelMask=-32, deviceProfileStates [{profileName=Switch, qualityLevelMask=4294967295, setQualityLevelMask=4294967295}]}}})
      True Float=0.75
      False Float=0.1
     )
    )
    ScaleA=true
    ScaleRGB=false
   )
   &ScaleRibbonWidth(Ribbon Width Scale=&FloatFromCurve(FloatCurve=DataInterface() {propertyValues {Curve {keys [<at 0 1 RCIM_Cubic>, {arriveTangent=-1.6891, interpMode=RCIM_Cubic, leaveTangent=-1.6891, tangentMode=RCTM_User, time=1, value=0}]}}}))
   /SetVariables(Particles.Lines COlor:LinearColor=&MakeLinearColorFromVectorAndFloat(
     "Vector (RGB)"=&MakeVectorFromLinearColorRGB(Color=$Particles.Initial.Lines COlor)
     "Float (Alpha)"=&FloatFromCurve(FloatCurve=DataInterface() {propertyValues {Curve {keys [{arriveTangent=-3.0635, interpMode=RCIM_Cubic, leaveTangent=-3.0635, tangentMode=RCTM_User, time=0, value=1}, <at 1 0 RCIM_Cubic>]}}})
    )) {isSetParameters=true, module=SetVariables_#3}
  ]
  rendererClasses [&NiagaraRibbonRendererProperties]
  renderers [
   {rendererClass=&NiagaraRibbonRendererProperties}
   {rendererClass=&NiagaraRibbonRendererProperties, rendererIndex=1}
  ]
 }
 Trail_FE {
  emitterUpdateScript [
   &EmitterState(
    Enable Distance Culling=true
    Module.bApplySpawnCountScaleByDistance=false
    Module.bUseMaxDistance=false
    Module.bUseMinDistance=false
    Scalability Mode=::NewEnumerator1
    Scale Spawn Count=true
   )
   &SpawnRate(
    SpawnRate=&MakeCustomFloatFromBool(Bool=$User.bIsFrontEndPreview, True Float=100, False Float=0)
    Use Spawn Probability=false
   )
  ]
  particleSpawnScript [
   &InitializeRibbon(
    Color=<rgba 50 6.6667 0 1>
    Lifetime=0.5
    Module.Write Color=true
    Module.Write Lifetime=true
    Module.Write Mass=true
    Module.Write Material Random=true
    Module.Write Position=true
    Module.Write RibbonTwist=true
    Module.Write RibbonWidth=true
    Position:Vector3f=&Add_Vector(A=/Niagara/DynamicInputs/Helpers/SimulationPosition(), B=<0 5 0>)
    Ribbon Twist=-0.1
    Ribbon Width=25
   )
   /SetVariables() {isSetParameters=true, module=SetVariables_#1}
   /SetVariables(Particles.Lines COlor:LinearColor=<rgba 75 10 0 1>, Particles.Lines Twist=-0.35, Particles.Lines Width=20) {isSetParameters=true, module=SetVariables_#2}
   &AddVelocity(Velocity=<0 -200 0>)
  ]
  particleUpdateScript [
   &SolveForcesAndVelocity(Clamp Velocity=false, Limit Acceleration=false)
   &ParticleState()
   &ScaleColor(
    Scale Alpha=&FloatFromCurve(
     FloatCurve=DataInterface() {propertyValues {Curve {keys [
         {arriveTangent=5.1472, interpMode=RCIM_Cubic, leaveTangent=5.1044, tangentMode=RCTM_User, time=0, value=0}
         <at 0.5 1 RCIM_Cubic>
         {arriveTangent=-5.1044, interpMode=RCIM_Cubic, leaveTangent=-5.1044, tangentMode=RCTM_User, time=1, value=0}
        ]}}}
     Scale Curve=&MakeCustomFloatFromBool(
      Bool=&IsPlatformSetActive(New Platform Set=DataInterface() {propertyValues {Platforms {qualityLevelMask=-32, deviceProfileStates [{profileName=Switch, qualityLevelMask=4294967295, setQualityLevelMask=4294967295}]}}})
      True Float=0.75
      False Float=0.1
     )
    )
    ScaleA=true
    ScaleRGB=false
   )
   &ScaleRibbonWidth(Ribbon Width Scale=&FloatFromCurve(FloatCurve=DataInterface() {propertyValues {Curve {keys [<at 0 1 RCIM_Cubic>, {arriveTangent=-1.6891, interpMode=RCIM_Cubic, leaveTangent=-1.6891, tangentMode=RCTM_User, time=1, value=0}]}}}))
   /SetVariables(Particles.Lines COlor:LinearColor=&MakeLinearColorFromVectorAndFloat(
     "Vector (RGB)"=&MakeVectorFromLinearColorRGB(Color=$Particles.Initial.Lines COlor)
     "Float (Alpha)"=&FloatFromCurve(FloatCurve=DataInterface() {propertyValues {Curve {keys [{arriveTangent=-3.0635, interpMode=RCIM_Cubic, leaveTangent=-3.0635, tangentMode=RCTM_User, time=0, value=1}, <at 1 0 RCIM_Cubic>]}}})
    )) {isSetParameters=true, module=SetVariables_#3}
  ]
  rendererClasses [&NiagaraRibbonRendererProperties]
  renderers [
   {rendererClass=&NiagaraRibbonRendererProperties}
   {rendererClass=&NiagaraRibbonRendererProperties, rendererIndex=1}
  ]
 }
}
summary {
 systemName=NS_EmoteRhymeLock_WheelTrail
 emitters [
  {emitterName=Trail, rendererClasses [&NiagaraRibbonRendererProperties]}
  {emitterName=Trail_FE, rendererClasses [&NiagaraRibbonRendererProperties]}
 ]
 userVariables [
  User.bisFrontEnd {type=NiagaraBool}
  User.bIsFrontEndPreview {type=NiagaraBool}
  User.bIsInWater {type=NiagaraBool}
  User.WaterDepth {type=NiagaraFloat, defaultValue=0, description=""}
 ]
}
