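# Niagara burst effect for the jump in the RhymeLock emote, with land and water variants switched by User.bIsInWater.
# A brief glowing ground ring mesh with a light and a radial smoke puff cooling from hot color to dust always fire, with the smoke's color curve changing in water.
# On land it adds orange streaks and colliding rocks.
# In water it swaps these for mesh bubbles and a delayed foam mesh constrained to the water plane.
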NiagaraSystem /BRCosmetics/Effects/Fort_Effects/Emotes/RhymeLock/NS_Emote_RhymeLock_Jump
dependencies {
 usedDataInterfaces [&NiagaraDataInterfaceCurve, &NiagaraDataInterfaceColorCurve]
 usedDynamicInputs []
 usedModules [&SystemState]
}
emitters {
 Bubbles {
  emitterUpdateScript [
   &EmitterState()
   &SpawnRate(
    SpawnRate=&Multiply_Float(A=&MakeCustomFloatFromBool(Bool=$User.bIsInWater, True Float=1, False Float=0), B=25)
    Use Spawn Probability=false
   ) {module=SpawnRate001}
   &SpawnBurst_Instantaneous(
    Spawn Count=&MakeCustomIntFromBool(Bool=$User.bIsInWater, True INT=&RandomRangeInt(Minimum=70, Maximum=50), False INT=0)
    Use Spawn Probability=false
   )
  ]
  particleSpawnScript [
   /Niagara/Modules/Spawn/Initialization/InitializeParticle(
    Lifetime=&UniformRangedFloat(Minimum=0.15, Maximum=0.2)
    Module.Write Color=true
    Module.Write Lifetime=true
    Module.Write Mass=true
    Module.Write Material Random=true
    Module.Write Position=false
    Module.Write Scale=false
    Module.Write SpriteRotation=false
    Module.Write SpriteSize=false
    Position=&SimulationPosition()
   )
   &InitialMeshOrientation(
    Orientation Vector=/Niagara/DynamicInputs/Random/RandomVector(
     Vector Scale:NiagaraFloat=1
     Constraint Vector:Vector3f=<0 0 1>
     Randomness Mode:ENiagaraRandomnessMode=::NewEnumerator0
     Random Seed:NiagaraInt32=0
     Fixed Random Seed:NiagaraBool=false
    )
    Use Orientation Vector=true
    Use Rotation Vector=false
   ) {module=InitialMeshOrientation001}
   &AddVelocity(Velocity=<0 -150 25>) {enabled=false}
   &InheritVelocity(Module.Limit Speed=true, Velocity Limit=1000, Velocity Scale=<0.5 0.5 0>) {enabled=false}
   &SystemLocation(MaskSpawnGroup=false, Offset=<0 -50 0>) {enabled=false}
   &SystemLocation(
    MaskSpawnGroup=false
    Offset=&Lerp_Vector(A=<0 -50 0>, B=<0 50 0>, Alpha=&UniformAOrBFloat(A=0))
   ) {enabled=false, module=SystemLocation001}
   &AddVelocityInCone(
    Cone Angle=178
    Cone Axis=<0 0 1>
    Module.Use Velocity Falloff On Cone Axis=false
    Velocity Distribution Along Cone Axis=1
    Velocity Strength=500
   )
  ]
  particleUpdateScript [
   &ParticleState()
   &GravityForce(Gravity=<0 0 50>)
   &SolveForcesAndVelocity(Clamp Velocity=false, Limit Acceleration=false)
   &ScaleColor(
    Scale Alpha=&FloatFromCurve(FloatCurve=DataInterface() {propertyValues {Curve {keys [<at 0 -1>,<at 0.05 0>,<at 0.7904 0.0132>,<at 1.0011 -0.9961>]}}})
    ScaleA=true
    ScaleRGB=true
   )
   &ScaleMeshSize(Scale Factor=&VectorFromFloat(Value=&FloatFromCurve(FloatCurve=DataInterface() {propertyValues {Curve {keys [
          {interpMode=RCIM_Cubic, tangentMode=RCTM_User, time=-0.0002, value=0.0422}
          {arriveTangent=0.8781, interpMode=RCIM_Cubic, leaveTangent=0.8781, tangentMode=RCTM_User, time=0.1611, value=0.6343}
          <at 0.9903 0.9119>
         ]}}})))
   &DynamicMaterialParameters(
    Index 0 Param 1=1
    Index 0 Param 2=1
    Index 0 Param 3=&FloatFromCurve(FloatCurve=DataInterface() {propertyValues {Curve {keys [<at 0 0.5>,<at 1 5>]}}}, Scale Curve=32)
    Index 0 Param 4=&FloatFromCurve(FloatCurve=DataInterface() {propertyValues {Curve {keys [<at 0 1>,<at 1 0>]}}}, Scale Curve=12)
    Module.Param0WriteEnabled=true
    Module.Param1WriteEnabled=true
    Module.Param2WriteEnabled=true
    Module.Param3WriteEnabled=true
   )
  ]
  rendererClasses [&NiagaraMeshRendererProperties]
  renderers [{rendererClass=&NiagaraMeshRendererProperties, rendererIndex=0}]
 }
 FN_Water_Foam_SnapToSurface001 {
  emitterUpdateScript [
   &EmitterState()
   &SpawnBurst_Instantaneous(
    Spawn Count=&Multiply_Int(A=&MakeCustomIntFromBool(Bool=$User.bIsInWater, True INT=1, False INT=0), B=1)
    Spawn Time=0.2
    Use Spawn Probability=false
   )
  ]
  particleSpawnScript [
   &InitializeParticle(
    Lifetime=&UniformRangedFloat(Minimum=0.8, Maximum=1)
    Mass=&UniformRangedFloat(Minimum=0.6, Maximum=1.2)
    Mesh Scale=&VectorFromFloat(Value=&Multiply_Float(A=225, B=$User.EffectScale))
    Module.Write Color=true
    Module.Write Lifetime=true
    Module.Write Mass=true
    Module.Write Material Random=true
    Module.Write Position=false
    Module.Write Scale=true
    Module.Write SpriteRotation=true
    Module.Write SpriteSize=false
    Position=&SimulationPosition()
    Sprite Rotation=&NormalizedAngleToDegrees(NormalizedAngle=&UniformRangedFloat(Minimum=0, Maximum=1))
   )
   &DynamicMaterialParameters(
    Index 1 Param 3=200
    Index 3 Param 1=$User.WaterBodyIndex
    Index 3 Param 2=$User.WaterDepth
    Module.Param0WriteEnabled1=false
    Module.Param0WriteEnabled3=true
    Module.Param1WriteEnabled1=false
    Module.Param1WriteEnabled3=true
    Module.Param2WriteEnabled1=true
    Module.Param2WriteEnabled3=true
    Module.Param3WriteEnabled1=false
    Module.Param3WriteEnabled3=true
    Write Parameter Index 0=false
    Write Parameter Index 1=true
    Write Parameter Index 3=true
   )
   &InitializeWaterSnapToOwner()
   &InheritVelocity(Module.Limit Speed=true)
   &InitialMeshOrientation(
    Orientation Vector=<0 0 0>
    Rotation=&UniformRangedVector(Minimum=<0 0 0>, Maximum=<0 0 360>)
    Use Orientation Vector=true
    Use Rotation Vector=true
   ) {module=InitialMeshOrientation001}
  ]
  particleUpdateScript [
   &ParticleState()
   &DynamicMaterialParameters(
    Index 0 Param 1=&FloatFromCurve(FloatCurve=DataInterface() {propertyValues {Curve {keys [<at -0.0046 1>,<at 1.0011 4>]}}})
    Index 0 Param 2=&FloatFromCurve(FloatCurve=DataInterface() {propertyValues {Curve {keys [{arriveTangent=0.6218, interpMode=RCIM_Cubic, leaveTangent=0.6218, tangentMode=RCTM_User, time=-0.0046, value=0.5}, <at 1.0011 1 RCIM_Cubic>]}}})
    Index 0 Param 3=&FloatFromCurve(FloatCurve=DataInterface() {propertyValues {Curve {keys [<at -0.0046 1>,<at 1.0011 0.5>]}}})
    Index 2 Param 1=&FloatFromCurve(FloatCurve=DataInterface() {propertyValues {Curve {keys [<at -0.0046 0.4>,<at 1.0011 1.5>]}}})
    Index 2 Param 2=&FloatFromCurve(FloatCurve=DataInterface() {propertyValues {Curve {keys [<at -0.0046 0.1>,<at 0.11 2.6>,<at 1.0011 0.3>]}}})
    Index 2 Param 3=1
    Index 2 Param 4=&FloatFromCurve(FloatCurve=DataInterface() {propertyValues {Curve {keys [
         {arriveTangent=-1.9991, interpMode=RCIM_Cubic, time=0, value=1}
         {arriveTangent=-2.7143, interpMode=RCIM_Cubic, leaveTangent=-2.7143, tangentMode=RCTM_User, time=0.2095, value=0.6901}
         <at 1 0 RCIM_Cubic>
        ]}}})
    Module.Param0WriteEnabled=true
    Module.Param0WriteEnabled2=true
    Module.Param1WriteEnabled=true
    Module.Param1WriteEnabled2=true
    Module.Param2WriteEnabled=true
    Module.Param2WriteEnabled2=true
    Module.Param3WriteEnabled=true
    Module.Param3WriteEnabled2=true
    Write Parameter Index 2=true
   ) {module=DynamicMaterialParameters001}
   &Drag(Drag=2, Use Linear Drag=true, Use Rotational Drag=false)
   &SolveForcesAndVelocity(Clamp Velocity=false, Limit Acceleration=false)
   &ScaleMeshSize(Scale Factor=&VectorFromFloat(Value=&FloatFromCurve(FloatCurve=DataInterface() {propertyValues {Curve {keys [
          {arriveTangent=26.4511, interpMode=RCIM_Cubic, leaveTangent=26.451, tangentMode=RCTM_User, time=-0.0046, value=0.6073}
          {arriveTangent=3.3733, interpMode=RCIM_Cubic, leaveTangent=3.3733, time=0.1268, value=2.5016}
          <at 1.0011 4 RCIM_Cubic>
         ]}}}))) {module=MeshSizeScale}
   &ConstrainPositionToPlane(
    Plane Normal:Vector3f=$Particles.WaterPlaneNormal
    Plane Position:NiagaraPosition=&Add_Vector(A=$User.WaterPlaneLocation, B=<0 0 0>)
   )
   &CameraOffset(Camera Offset Amount=10)
  ]
  rendererClasses [&NiagaraMeshRendererProperties]
  renderers [{rendererClass=&NiagaraMeshRendererProperties, rendererIndex=0}]
 }
 GroundRocks {
  emitterUpdateScript [
   &EmitterState()
   &SpawnBurst_Instantaneous(
    Spawn Count=&MakeCustomIntFromBool(Bool=$User.bIsInWater, True INT=0, False INT=&RandomRangeInt(Minimum=15, Maximum=12))
    Use Spawn Probability=false
   )
  ]
  particleSpawnScript [
   /SetVariables(Particles.MaterialRandom:NiagaraFloat=&RandomRangeFloat(Minimum=0, Maximum=1)) {isSetParameters=true}
   &InitializeParticle(
    Color=&LerpLinearColors(StartColor=<rgba 0.6 0.495 0.42 1>, EndColor=<rgba 0.6 0.495 0.42 1>, LerpFactor=$Particles.MaterialRandom)
    Lifetime=&Lerp_Float(A=0.4, B=1, Alpha=$Particles.MaterialRandom)
    Module.Write Color=true
    Module.Write Lifetime=true
    Module.Write Mass=true
    Module.Write Material Random=false
    Module.Write Position=false
    Module.Write Scale=false
    Module.Write SpriteRotation=true
    Module.Write SpriteSize=true
    Position=&SimulationPosition()
    Sprite Rotation=&RandomRangeFloat(Minimum=0, Maximum=360)
    Sprite Size=&Vector2DFromFloat(Value=&RandomRangeFloat(Minimum=10, Maximum=5))
   )
   &SphereLocation(
    MaskSpawnGroup=false
    Non Uniform Scale=<1 1 0>
    Offset=<0 0 5>
    Override Seed=false
    Sphere Radius=&Lerp_Float(A=10, B=50, Alpha=&Power(a=$Particles.MaterialRandom, b=1.5))
    Surface Only=false
    UseNonUniformScale=true
    UseOffset=true
    UseSphereOrientation=false
   )
   &AddVelocityFromPoint(
    Module.Use Falloff Distance=false
    Velocity Strength=&Lerp_Float(A=150, B=600, Alpha=&Power(a=$Particles.MaterialRandom, b=3))
   )
   &AddVelocity(Velocity=<0 0 200>)
  ]
  particleUpdateScript [
   &ParticleState()
   &GravityForce()
   &Drag(Use Linear Drag=true, Use Rotational Drag=false)
   &SpriteRotationRate(Rotation Rate=&Lerp_Float(A=100, B=650, Alpha=$Particles.MaterialRandom))
   &ScaleSpriteSize(Scale Factor=&Vector2DFromFloat(Value=&FloatFromCurve(FloatCurve=DataInterface() {propertyValues {Curve {keys [
          {arriveTangent=35.6763, interpMode=RCIM_Cubic, leaveTangent=3.7835, tangentMode=RCTM_User, time=0, value=0}
          <at 0.7 1 RCIM_Cubic>
          {interpMode=RCIM_Cubic, leaveTangent=0.6903, time=1, value=0}
         ]}}})))
   &Collision(
    Advanced Aging Rate=1
    Age Colliding Particles=false
    Module.Control Rest State Via Surface Normal=false
    Module.EnableMaxCollisionCount=false
    Module.Randomize Collision Normal Vector=false
    Restitution=0.3
   )
   /SetVariables(Particles.MaxCollisions:NiagaraBool=&SetBoolByIntComparison(A=$Particles.Number Of Collisions, B=1, Comparison Type=::NewEnumerator0)) {isSetParameters=true}
   &SolveForcesAndVelocity(Clamp Velocity=false, Limit Acceleration=false)
   /SetVariables(Particles.SpriteRotation:NiagaraFloat=&MakeCustomFloatFromBool(Bool=$Particles.MaxCollisions, True Float=0, False Float=$Particles.SpriteRotation)) {isSetParameters=true}
  ]
  rendererClasses [&NiagaraSpriteRendererProperties]
  renderers [{rendererClass=&NiagaraSpriteRendererProperties}]
 }
 Smoke {
  emitterUpdateScript [
   &EmitterState()
   &SpawnBurst_Instantaneous(Spawn Count=&RandomRangeInt(Minimum=20, Maximum=15), Use Spawn Probability=false)
  ]
  particleSpawnScript [
   /Niagara/Modules/Spawn/Initialization/V2/InitializeParticle(
    Color=&MakeLinearColorFromVectorAndFloat(
     "Vector (RGB)"=&Multiply_Vector(
      A=&MakeVectorFromLinearColorRGB(Color=<rgba 0.2344 0.2067 0.142 1>)
      B=&VectorFromFloat(Value=1)
     )
     "Float (Alpha)"=0.25
    )
    Lifetime=&RandomRangeFloat(Minimum=0.35, Maximum=0.5)
    Sprite Rotation Mode=::NewEnumerator2
    Sprite Size Mode=::NewEnumerator1
    Sprite UV Mode=::NewEnumerator3
    Uniform Sprite Size Max=80
    Uniform Sprite Size Min=50
    UsePositionOffset=false
    Write Material Random=true
   )
   &CylinderLocation(
    Cylinder Height=0
    Cylinder Radius=&Lerp_Float(A=0.1, B=0.1, Alpha=$Particles.MaterialRandom)
    MaskSpawnGroup=false
    Module.Enable Lathe Profile=false
    Module.Override Randomness=false
    Module.Surface Only=true
    Surface Only Band Thickness=0
    Use Endcaps In Surface Only Mode=false
   )
   &AddVelocityFromPoint(Module.Use Falloff Distance=false, Velocity Strength=&Lerp_Float(A=800, B=1000, Alpha=$Particles.MaterialRandom))
   &SubUVAnimation(SubUV Animation Mode:ENiagara_SubUVLookupMode=::NewEnumerator2)
   &AddVelocity(Velocity=&MakeVector(X=0, Y=0, Z=&RandomRangeFloat(Minimum=0, Maximum=500)))
  ]
  particleUpdateScript [
   &ParticleState()
   &Drag(Drag=6, Use Linear Drag=true, Use Rotational Drag=false)
   &SolveForcesAndVelocity(Clamp Velocity=false, Limit Acceleration=false)
   &ScaleSpriteSize(Scale Factor=&Vector2DFromFloat(Value=&FloatFromCurve(
      FloatCurve=DataInterface() {propertyValues {Curve {keys [
          {arriveTangent=0.0178, interpMode=RCIM_Cubic, time=0, value=0}
          {arriveTangent=0.5456, interpMode=RCIM_Cubic, leaveTangent=0.5456, tangentMode=RCTM_User, time=0.193, value=0.6844}
          {interpMode=RCIM_Cubic, leaveTangent=-0.5927, tangentMode=RCTM_User, time=1, value=1}
         ]}}}
      Scale Curve=2
     )))
   &DynamicMaterialParameters(
    Index 0 Param 1=&FloatFromCurve(FloatCurve=DataInterface() {propertyValues {Curve {keys [{arriveTangent=-2.3695, interpMode=RCIM_Cubic, leaveTangent=-2.3695, tangentMode=RCTM_User, time=0, value=0.8}, <at 1 0>]}}})
    Module.Param0WriteEnabled=true
    Module.Param1WriteEnabled=true
    Module.Param2WriteEnabled=true
    Module.Param3WriteEnabled=true
   )
   &Color(Color=&MakeCustomLinearColorFromBool(
     Bool=$User.bIsInWater
     True Color=&ColorFromCurve(ColorCurve=DataInterface() {propertyValues {
        AlphaCurve {keys [<at 0 1 RCIM_Cubic>, {arriveTangent=-1.9188, interpMode=RCIM_Cubic, leaveTangent=-1.9188, time=0.4086, value=1}, <at 0.5212 0 RCIM_Cubic>]}
        BlueCurve {keys [<at 0.2493 0 RCIM_Cubic>,<at 0.4993 0.142 RCIM_Cubic>]}
        GreenCurve {keys [<at 0.2493 2 RCIM_Cubic>,<at 0.4993 0.2067 RCIM_Cubic>]}
        RedCurve {keys [{arriveTangent=-122.0462, interpMode=RCIM_Cubic, time=0.2493, value=15}, <at 0.4993 0.2344 RCIM_Cubic>]}
       }})
     False Color=&ColorFromCurve(ColorCurve=DataInterface() {propertyValues {
        AlphaCurve {keys [
          <at 0 1 RCIM_Cubic>
          {arriveTangent=-1, interpMode=RCIM_Cubic, leaveTangent=-2.8407, tangentMode=RCTM_Break, time=0.4569, value=1}
          {arriveTangent=-2.5877, interpMode=RCIM_Cubic, leaveTangent=-2.5877, tangentMode=RCTM_User, time=1, value=0}
         ]}
        BlueCurve {keys [<at 0.0859 0.25 RCIM_Cubic>,<at 0.4634 0.142 RCIM_Cubic>]}
        GreenCurve {keys [<at 0.0859 0.8833 RCIM_Cubic>,<at 0.4634 0.2067 RCIM_Cubic>]}
        RedCurve {keys [{arriveTangent=-35.2756, time=0.0859, value=5}, <at 0.4634 0.2344>]}
       }})
    )) {module=Color001}
  ]
  rendererClasses [&NiagaraSpriteRendererProperties]
  renderers [{rendererClass=&NiagaraSpriteRendererProperties}]
 }
 SmokeGroundRings {
  emitterUpdateScript [
   &EmitterState()
   &SpawnBurst_Instantaneous(Spawn Count=1, Use Spawn Probability=false)
  ]
  particleSpawnScript [
   /Niagara/Modules/Spawn/Initialization/InitializeParticle(
    Color=<rgba 10 2.4 0.5 1>
    Lifetime=0.225
    Mesh Scale=&VectorFromFloat(Value=1.8)
    Module.Write Color=true
    Module.Write Lifetime=true
    Module.Write Mass=true
    Module.Write Material Random=true
    Module.Write Position=false
    Module.Write Scale=true
    Module.Write SpriteRotation=true
    Module.Write SpriteSize=true
    Position=&SimulationPosition()
    Sprite Size=<2.5 2.5>
   )
   &SystemLocation(MaskSpawnGroup=false, Offset=<0 0 5>)
  ]
  particleUpdateScript [
   &ParticleState()
   &DynamicMaterialParameters(
    Index 0 Param 2=1
    Index 0 Param 3=1
    Module.Param0WriteEnabled=false
    Module.Param1WriteEnabled=true
    Module.Param2WriteEnabled=true
    Module.Param3WriteEnabled=false
   )
   &ScaleSpriteSize(Scale Factor=&Vector2DFromFloat(Value=&FloatFromCurve(
      FloatCurve=DataInterface() {propertyValues {Curve {keys [
          {arriveTangent=35.6763, interpMode=RCIM_Cubic, leaveTangent=3.7835, tangentMode=RCTM_User, time=0.03, value=0}
          {interpMode=RCIM_Cubic, leaveTangent=0.6903, time=1.0011, value=1}
         ]}}}
      Scale Curve=50
     )))
   &ScaleColor(
    Scale Alpha=&FloatFromCurve(FloatCurve=DataInterface() {propertyValues {Curve {keys [
         {arriveTangent=-5.9613, interpMode=RCIM_Cubic, leaveTangent=-5.9613, tangentMode=RCTM_User, time=0.5, value=1}
         {arriveTangent=-0.173, interpMode=RCIM_Cubic, leaveTangent=-1.365, tangentMode=RCTM_User, time=1.0011, value=0.002}
        ]}}})
    ScaleA=true
    ScaleRGB=false
   )
   &ScaleMeshSize(Scale Factor=&VectorFromFloat(Value=&FloatFromCurve(FloatCurve=DataInterface() {propertyValues {Curve {keys [<at 0 0>,<at 1 1>]}}})))
   &FN_Light_Attributes(Color=<rgba 5 2 1.25 1>, Create Color Attribute=true, Create Exponent Attribute=true, Create Radius Attribute=true, Exponent=1, Radius=50)
  ]
  rendererClasses [&NiagaraMeshRendererProperties, &NiagaraLightRendererProperties]
  renderers [
   {rendererClass=&NiagaraMeshRendererProperties, rendererIndex=0}
   {rendererClass=&NiagaraLightRendererProperties, rendererIndex=1}
  ]
 }
 Streaks001 {
  emitterUpdateScript [
   &EmitterState()
   &SpawnBurst_Instantaneous(
    Spawn Count=&MakeCustomIntFromBool(Bool=$User.bIsInWater, True INT=0, False INT=&RandomRangeInt(Minimum=20, Maximum=35))
    Use Spawn Probability=false
   )
  ]
  particleSpawnScript [
   &InitializeParticle(
    Color=&MakeLinearColorFromVectorAndFloat("Vector (RGB)"=&MakeVectorFromLinearColorRGB(Color=<rgba 15 3.75 0 1>), "Float (Alpha)"=1)
    Lifetime=&UniformRangedFloat(Minimum=0.2, Maximum=0.1)
    Module.Write Color=true
    Module.Write Lifetime=true
    Module.Write Mass=true
    Module.Write Material Random=true
    Module.Write Position=false
    Module.Write Scale=false
    Module.Write SpriteRotation=true
    Module.Write SpriteSize=true
    Position=&SimulationPosition()
    Sprite Size=&Multiply_Vector2(
     A=&UniformRangedVector2D(Minimum=<15 80>, Maximum=<20 160>)
     B=&Vector2DFromFloat(Value=0.3)
    )
   )
   &CylinderLocation(
    Cylinder Height=0
    Cylinder Radius=1
    MaskSpawnGroup=false
    Module.Enable Lathe Profile=false
    Module.Override Randomness=false
    Module.Surface Only=true
    Surface Only Band Thickness=1
    Use Endcaps In Surface Only Mode=false
   )
   &AddVelocity(Velocity=&MakeVector(X=0, Y=0, Z=&UniformRangedFloat(Minimum=0, Maximum=300)))
   &AddVelocityFromPoint(
    Module.Use Falloff Distance=false
    Velocity Origin:Vector3f=&SimulationPosition()
    Velocity Strength=&RandomRangeFloat(Minimum=800, Maximum=3000)
   )
   &ApplyInitialForces()
   &DynamicMaterialParameters(
    Index 0 Param 1=&UniformRangedFloat(Minimum=0, Maximum=1)
    Index 0 Param 2=&UniformRangedFloat(Minimum=1, Maximum=2)
    Index 0 Param 3=1
    Index 0 Param 4=1
    Module.Param0WriteEnabled=true
    Module.Param1WriteEnabled=true
    Module.Param2WriteEnabled=true
    Module.Param3WriteEnabled=true
   )
  ]
  particleUpdateScript [
   &ParticleState()
   &Drag(
    Drag=&RandomRangeFloat(Minimum=7, Maximum=10, Evaluation Type=::NewEnumerator1)
    Use Linear Drag=true
    Use Rotational Drag=false
   )
   &SolveForcesAndVelocity(Clamp Velocity=false, Limit Acceleration=false)
   &ScaleColor(
    Scale Alpha=&FloatFromCurve(FloatCurve=DataInterface() {propertyValues {Curve {keys [<at 0 0 RCIM_Cubic>, {arriveTangent=0.002, interpMode=RCIM_Cubic, leaveTangent=0.002, time=0.5, value=1}, <at 1.0011 0.002 RCIM_Cubic>]}}})
    Scale RGB=&MakeVectorFromLinearColorRGB(Color=&ColorFromCurve(ColorCurve=DataInterface() {propertyValues {
        AlphaCurve {keys [<at 0 1 RCIM_Cubic>,<at 1 0 RCIM_Cubic>]}
        BlueCurve {keys [<at 0 1.14 RCIM_Cubic>,<at 1 1 RCIM_Cubic>]}
        GreenCurve {keys [<at 0 1.842 RCIM_Cubic>,<at 1 1 RCIM_Cubic>]}
        RedCurve {keys [<at 0 3 RCIM_Cubic>,<at 1 1 RCIM_Cubic>]}
       }}))
    ScaleA=true
    ScaleRGB=false
   )
  ]
  rendererClasses [&NiagaraSpriteRendererProperties]
  renderers [{rendererClass=&NiagaraSpriteRendererProperties}]
 }
}
summary {
 systemName=NS_Emote_RhymeLock_Jump
 emitters [
  {
   emitterName=SmokeGroundRings
   rendererClasses [&NiagaraMeshRendererProperties, &NiagaraLightRendererProperties]
  }
  {emitterName=Streaks001, rendererClasses [&NiagaraSpriteRendererProperties]}
  {emitterName=Smoke, rendererClasses [&NiagaraSpriteRendererProperties]}
  {emitterName=GroundRocks, rendererClasses [&NiagaraSpriteRendererProperties]}
  {emitterName=Bubbles, rendererClasses [&NiagaraMeshRendererProperties]}
  {emitterName=FN_Water_Foam_SnapToSurface001, rendererClasses [&NiagaraMeshRendererProperties]}
 ]
 userVariables [
  User.bIsInWater {type=NiagaraBool}
  User.EffectScale {type=NiagaraFloat, defaultValue=1, description=""}
  User.WaterBodyIndex {type=NiagaraFloat, defaultValue=0, description=""}
  User.WaterDepth {type=NiagaraFloat, defaultValue=0, description=""}
  User.WaterPlaneLocation {type=Vector3f, description="", defaultValue {struct=Vector3f, value=<0 0 0>}}
  User.WaterPlaneNormal {type=Vector3f, description="", defaultValue {struct=Vector3f, value=<0 0 0>}}
  User.WaveAttenuationFactor {type=NiagaraFloat, defaultValue=0, description=""}
  User.WaveTime {type=NiagaraFloat, defaultValue=0, description=""}
 ]
}
